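# Physics asset for the F_MED_DrapeUltra female face-accessory mesh (no-backpack variant), covering hair and pauldron physics.
# Capsules cover pelvis, spine, thorax, neck, head, clavicles and upper arms, with pauldron spheres constrained to the upper arms.
# Three long hair chains and two front side strands hang from head-linked base bones with narrow angular limits.
# Free-rotation constraints link body parts, pauldrons and hair bases, and tie the center chain to both side chains near the tips.
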PhysicsAsset /BRCosmetics/Characters/Player/Female/Medium/Bodies/F_MED_DrapeUltra/Meshes/Parts/F_MED_DrapeUltra_FaceAcc_WITHOUT_Backpack_Physics
bodies {
 base_hair_cr [{shapeType=Capsule, length=0.505, radius=0.505, shapeName=base_hair_cr_capsule, rotation=<rot 0 -90 0>}]
 base_hair_front_side_02_03_lf [{shapeType=Capsule, length=0.505, radius=0.505, shapeName=base_hair_front_side_02_03_lf_capsule, rotation=<rot 0 -90 0>}]
 base_hair_front_side_02_03_rt [{shapeType=Capsule, length=0.505, radius=0.505, shapeName=base_hair_front_side_02_03_rt_capsule, rotation=<rot 0 -90 0>}]
 base_hair_lf [{shapeType=Capsule, length=0.505, radius=0.505, shapeName=base_hair_lf_capsule, rotation=<rot 0 -90 0>}]
 base_hair_master [{shapeType=Capsule, length=23.0586, radius=8, shapeName=hair_master, center=<4.5566 0 -6.4535>, rotation=<rot 0 0 90>}]
 base_hair_rt [{shapeType=Capsule, length=0.505, radius=0.505, shapeName=base_hair_rt_capsule, rotation=<rot 0 -90 0>}]
 clavicle_l [
  {shapeType=Capsule, length=5.1948, radius=4.7851, shapeName=clavicle_l_capsule, center=<8.8633 0.0327 -0.656>, rotation=<rot 90 0 0>}
  {
   shapeType=Capsule
   length=8.9447
   radius=4.7782
   shapeName=base_pauldron_l_capsule
   center=<13.4818 -0.5606 -3.5373>
   rotation=<rot 51.5465 0.0001 -0.0001>
  }
 ]
 clavicle_r [
  {shapeType=Capsule, length=4.3991, radius=4.802, shapeName=clavicle_l_capsule, center=<-9.3665 -0.0327 0.7413>, rotation=<rot 90 0 0>}
  {
   shapeType=Capsule
   length=7.1599
   radius=5.1281
   shapeName=base_pauldron_l_capsule
   center=<-13.4818 0.3182 3.5373>
   rotation=<rot 51.5465 0.0001 -0.0001>
  }
 ]
 dyn_hair_01_cr [{shapeType=Capsule, length=11.5986, radius=7, shapeName=dyn_hair_01_rt_capsule, center=<5.7993 0 11.625>, rotation=<rot 90 0 0>}]
 dyn_hair_01_lf [{shapeType=Capsule, length=11.5986, radius=7, shapeName=dyn_hair_01_rt_capsule, center=<5.7993 3.25 5.5>, rotation=<rot 90 0 0>}]
 dyn_hair_01_rt [{shapeType=Capsule, length=11.5986, radius=7, shapeName=dyn_hair_01_rt_capsule, center=<5.7993 -3.5 5.25>, rotation=<rot 90 0 0>}]
 dyn_hair_02_cr [{shapeType=Capsule, length=6.9, radius=7, shapeName=dyn_hair_02_rt_capsule, center=<3.45 0 12>, rotation=<rot 90 0 0>}]
 dyn_hair_02_lf [{shapeType=Capsule, length=6.9, radius=7, shapeName=dyn_hair_02_rt_capsule, center=<3.45 3.5 5.5>, rotation=<rot 90 0 0>}]
 dyn_hair_02_rt [{shapeType=Capsule, length=6.9, radius=7, shapeName=dyn_hair_02_rt_capsule, center=<3.45 -3.5 5.5>, rotation=<rot 90 0 0>}]
 dyn_hair_03_cr [{shapeType=Capsule, length=6.9, radius=7, shapeName=dyn_hair_03_cr_capsule, center=<3.45 0 3>, rotation=<rot 90 0 0>}]
 dyn_hair_03_lf [{shapeType=Capsule, length=6.9, radius=7, shapeName=dyn_hair_02_rt_capsule, center=<3.45 0 3>, rotation=<rot 90 0 0>}]
 dyn_hair_03_rt [{shapeType=Capsule, length=6.9, radius=7, shapeName=dyn_hair_02_rt_capsule, center=<3.45 0 3>, rotation=<rot 90 0 0>}]
 dyn_hair_04_cr [{shapeType=Capsule, length=7.42, radius=7, shapeName=dyn_hair_04_cr_capsule, center=<3.71 0 3>, rotation=<rot 90 0 0>}]
 dyn_hair_04_lf [{shapeType=Capsule, length=7.42, radius=7, shapeName=dyn_hair_04_cr_capsule, center=<3.71 0 3>, rotation=<rot 90 0 0>}]
 dyn_hair_04_rt [{shapeType=Capsule, length=7.42, radius=7, shapeName=dyn_hair_04_cr_capsule, center=<3.71 0 3>, rotation=<rot 90 0 0>}]
 dyn_hair_05_cr [{shapeType=Capsule, length=9.3229, radius=7, shapeName=dyn_hair_05_cr_capsule, center=<4.6614 0 3>, rotation=<rot 90 0 0>}]
 dyn_hair_05_lf [{shapeType=Capsule, length=9.3229, radius=7, shapeName=dyn_hair_05_cr_capsule, center=<4.6614 0 3>, rotation=<rot 90 0 0>}]
 dyn_hair_05_rt [{shapeType=Capsule, length=9.3229, radius=7, shapeName=dyn_hair_05_cr_capsule, center=<4.6614 0 3>, rotation=<rot 90 0 0>}]
 dyn_hair_06_cr [{shapeType=Capsule, length=9.3229, radius=7, shapeName=dyn_hair_06_cr_capsule, center=<4.6614 0 3>, rotation=<rot 90 0 0>}]
 dyn_hair_06_lf [{shapeType=Capsule, length=9.3229, radius=7, shapeName=dyn_hair_06_cr_capsule, center=<4.6614 0 3>, rotation=<rot 90 0 0>}]
 dyn_hair_06_rt [{shapeType=Capsule, length=9.3229, radius=7, shapeName=dyn_hair_06_cr_capsule, center=<4.6614 0 3>, rotation=<rot 90 0 0>}]
 dyn_hair_07_cr [{shapeType=Capsule, length=5.625, radius=7, shapeName=dyn_hair_07_cr_capsule, center=<2.8125 0 3>, rotation=<rot 90 0 0>}]
 dyn_hair_07_lf [{shapeType=Capsule, length=5.625, radius=7, shapeName=dyn_hair_07_cr_capsule, center=<2.8125 0 3>, rotation=<rot 90 0 0>}]
 dyn_hair_07_rt [{shapeType=Capsule, length=5.625, radius=7, shapeName=dyn_hair_07_cr_capsule, center=<2.8125 0 3>, rotation=<rot 90 0 0>}]
 dyn_hair_08_cr [{shapeType=Capsule, length=5.625, radius=7, shapeName=dyn_hair_07_cr_capsule, center=<2.8125 0 3>, rotation=<rot 90 0 0>}]
 dyn_hair_front_side_02_03_lf [{
   shapeType=Capsule
   length=4.8061
   radius=2.0725
   shapeName=dyn_hair_front_side_02_05_lf_capsule
   center=<2.5434 0 -0.1272>
   rotation=<rot 86.2989 -179.9999 -179.9999>
  }]
 dyn_hair_front_side_02_03_rt [{
   shapeType=Capsule
   length=4.6941
   radius=1.9981
   shapeName=dyn_hair_front_side_02_05_lf_capsule
   center=<2.5823 0.0811 0.1388>
   rotation=<rot 88.5721 -5.2471 -7.1802>
  }]
 dyn_hair_front_side_02_04_lf [{
   shapeType=Capsule
   length=3.9387
   radius=2.0613
   shapeName=dyn_hair_front_side_02_05_lf_capsule
   center=<2.5 0 0>
   rotation=<rot 85.5155 -179.9997 -179.9997>
  }]
 dyn_hair_front_side_02_04_rt [{
   shapeType=Capsule
   length=4.739
   radius=1.8205
   shapeName=dyn_hair_front_side_02_05_lf_capsule
   center=<2.7126 -0.0682 -0.0116>
   rotation=<rot 88.144 28.368 28.4055>
  }]
 dyn_hair_front_side_02_05_lf [{
   shapeType=Capsule
   length=4.1208
   radius=1.6078
   shapeName=dyn_hair_front_side_02_05_lf_capsule
   center=<2.0195 0 0>
   rotation=<rot 86.6712 -0.0002 -0.0002>
  }]
 dyn_hair_front_side_02_05_rt [{
   shapeType=Capsule
   length=3.9639
   radius=1.6111
   shapeName=dyn_hair_front_side_02_05_lf_capsule
   center=<2.2053 0.2405 -0.2589>
   rotation=<rot 89.7224 179.9969 179.9969>
  }]
 dyn_hair_front_side_02_06_lf [{
   shapeType=Capsule
   length=5.8401
   radius=1.5
   shapeName=dyn_hair_front_side_02_06_lf_capsule
   center=<2.8651 0 0.12>
   rotation=<rot 86.1349 0 1.2114>
  }]
 dyn_hair_front_side_02_06_rt [{
   shapeType=Capsule
   length=4.7853
   radius=1.5386
   shapeName=dyn_hair_front_side_02_06_lf_capsule
   center=<2.2285 -0.0014 -0.0256>
   rotation=<rot 87.6212 -14.1086 -8.3101>
  }]
 dyn_hair_front_side_02_07_lf [{
   shapeType=Capsule
   length=6.7909
   radius=1.3
   shapeName=dyn_hair_front_side_02_07_lf_capsule
   center=<3.3555 0 -0.1756>
   rotation=<rot 88.9802 0 0>
  }]
 dyn_hair_front_side_02_07_rt [{
   shapeType=Capsule
   length=5.6874
   radius=1.2512
   shapeName=dyn_hair_front_side_02_07_lf_capsule
   center=<2.8211 0.0001 -0.2635>
   rotation=<rot 89.1478 -2.1601 -0.0003>
  }]
 dyn_hair_front_side_02_08_lf [{shapeType=Capsule, length=5.1262, radius=0.9547, shapeName=dyn_hair_front_side_02_07_lf_capsule, center=<3 0 0>, rotation=<rot 90 0 0>}]
 dyn_hair_front_side_02_08_rt [{
   shapeType=Capsule
   length=8.0578
   radius=1.125
   shapeName=dyn_hair_front_side_02_07_lf_capsule
   center=<3.1484 0.1694 0>
   rotation=<rot 90 4.8973 0>
  }]
 dyn_hair_master [{shapeType=Capsule, length=0.505, radius=0.505, shapeName=dyn_hair_master_capsule, rotation=<rot 0 -90 0>}]
 dyn_pauldron_fr_l [{shapeType=Sphere, radius=7.7932, shapeName=dyn_pauldron_l_sphere, center=<-0.9392 -0.4536 0.9924>}]
 dyn_pauldron_fr_r [{shapeType=Sphere, radius=7.9369, shapeName=dyn_pauldron_l_sphere, center=<0 0.501 0.2552>}]
 dyn_pauldron_l [{shapeType=Sphere, radius=6.8036, shapeName=dyn_pauldron_l_sphere, center=<0 0 5.9863>}]
 dyn_pauldron_r [{shapeType=Sphere, radius=7.2109, shapeName=dyn_pauldron_r_sphere, center=<0 0 -7.644>}]
 head [{shapeType=Capsule, length=6.3601, radius=10.0804, shapeName=head_capsule, center=<3.0622 0.1061 0>, rotation=<rot 90 12.1159 0>}]
 neck_01 [{shapeType=Capsule, length=2.125, radius=4.2139, shapeName=neck_02_capsule, center=<3.1095 1.3383 0>}]
 neck_02 [{shapeType=Capsule, length=2.125, radius=4.2139, shapeName=neck_02_capsule, center=<3.5199 2.0658 0>}]
 pelvis [{shapeType=Capsule, length=7, radius=11, shapeName=pelvis_capsule, rotation=<rot 0 -90 0>}]
 spine_01 [{shapeType=Capsule, length=9, radius=10.5, shapeName=spine_01_capsule, rotation=<rot 0 -90 0>}]
 spine_02 [{shapeType=Capsule, length=9.2772, radius=7.0928, shapeName=spine_02_capsule, rotation=<rot 0 -90 0>}]
 spine_03 [{shapeType=Capsule, length=9.1101, radius=6.6403, shapeName=spine_03_capsule, rotation=<rot 0 -90 0>}]
 spine_04 [
  {shapeType=Capsule, length=11.6103, radius=6.2164, shapeName=spine_04_capsule, center=<-0.068 -2.9239 0>}
  {shapeType=Capsule, length=11, radius=5.5583, shapeName=spine_04_capsule, center=<2.4757 8.1388 -0.0162>, rotation=<rot 0 -90 0>}
  {shapeType=Capsule, length=10.4022, radius=4.7101, shapeName=spine_04_capsule, center=<6.7937 4.6415 0.3329>, rotation=<rot 0 -90 0>}
  {
   shapeType=Capsule
   length=10.1015
   radius=5.0109
   shapeName=spine_04_capsule
   center=<6.423 3.387 -5.8643>
   rotation=<rot 90 -82.704 -30.0197>
  }
  {
   shapeType=Capsule
   length=10.1015
   radius=5.0109
   shapeName=spine_04_capsule
   center=<6.2179 3.9258 6.556>
   rotation=<rot 70.6108 152.5863 -152.1304>
  }
 ]
 spine_05 [{shapeType=Capsule, length=11.7473, radius=5.754, shapeName=spine_05_capsule, center=<2.2882 -0.9753 0>, rotation=<rot 0 -90 0>}]
 thorax [{shapeType=Capsule, length=15, radius=12.7164, shapeName=spine_04_capsule, center=<-0.5242 -8.2937 0>}]
 upperarm_l [{shapeType=Capsule, length=26, radius=4.25, shapeName=upperarm_l_capsule, center=<13 0 0>, rotation=<rot 90 0 0>}]
 upperarm_r [{shapeType=Capsule, length=26, radius=4.25, shapeName=upperarm_r_capsule, center=<-13 0 0>, rotation=<rot 90 0 0>}]
}
constraints [
 {
  bone1Name=spine_05
  bone2Name=spine_04
  swing1LimitDegrees=45
  swing1Motion=Free
  swing2LimitDegrees=45
  swing2Motion=Free
  twistLimitDegrees=45
  twistMotion=Free
 }
 {
  bone1Name=clavicle_l
  bone2Name=spine_05
  swing1LimitDegrees=45
  swing1Motion=Free
  swing2LimitDegrees=45
  swing2Motion=Free
  twistLimitDegrees=45
  twistMotion=Free
 }
 {
  bone1Name=clavicle_r
  bone2Name=spine_05
  swing1LimitDegrees=45
  swing1Motion=Free
  swing2LimitDegrees=45
  swing2Motion=Free
  twistLimitDegrees=45
  twistMotion=Free
 }
 {
  bone1Name=neck_02
  bone2Name=neck_01
  swing1LimitDegrees=45
  swing1Motion=Free
  swing2LimitDegrees=45
  swing2Motion=Free
  twistLimitDegrees=45
  twistMotion=Free
 }
 {
  bone1Name=base_hair_cr
  bone2Name=head
  swing1LimitDegrees=45
  swing1Motion=Free
  swing2LimitDegrees=45
  swing2Motion=Free
  twistLimitDegrees=45
  twistMotion=Free
 }
 {bone1Name=dyn_hair_01_cr, bone2Name=base_hair_cr, swing1LimitDegrees=9, swing2LimitDegrees=10, twistLimitDegrees=12}
 {bone1Name=dyn_hair_02_cr, bone2Name=dyn_hair_01_cr, swing1LimitDegrees=9, swing2LimitDegrees=45, twistLimitDegrees=12}
 {bone1Name=dyn_hair_03_cr, bone2Name=dyn_hair_02_cr, swing1LimitDegrees=9, swing2LimitDegrees=45, twistLimitDegrees=12}
 {bone1Name=dyn_hair_04_cr, bone2Name=dyn_hair_03_cr, swing1LimitDegrees=9, swing2LimitDegrees=30, twistLimitDegrees=12}
 {bone1Name=dyn_hair_05_cr, bone2Name=dyn_hair_04_cr, swing1LimitDegrees=9, swing2LimitDegrees=24, twistLimitDegrees=12}
 {bone1Name=dyn_hair_06_cr, bone2Name=dyn_hair_05_cr, swing1LimitDegrees=18, swing2LimitDegrees=24, twistLimitDegrees=12}
 {bone1Name=dyn_hair_07_cr, bone2Name=dyn_hair_06_cr, swing1LimitDegrees=18, swing2LimitDegrees=24, twistLimitDegrees=12}
 {
  bone1Name=base_hair_lf
  bone2Name=head
  swing1LimitDegrees=45
  swing1Motion=Free
  swing2LimitDegrees=45
  swing2Motion=Free
  twistLimitDegrees=45
  twistMotion=Free
 }
 {bone1Name=dyn_hair_01_lf, bone2Name=base_hair_lf, swing1LimitDegrees=9, swing2LimitDegrees=10, twistLimitDegrees=12}
 {bone1Name=dyn_hair_02_lf, bone2Name=dyn_hair_01_lf, swing1LimitDegrees=9, swing2LimitDegrees=45, twistLimitDegrees=12}
 {bone1Name=dyn_hair_03_lf, bone2Name=dyn_hair_02_lf, swing1LimitDegrees=9, swing2LimitDegrees=45, twistLimitDegrees=12}
 {bone1Name=dyn_hair_04_lf, bone2Name=dyn_hair_03_lf, swing1LimitDegrees=9, swing2LimitDegrees=30, twistLimitDegrees=12}
 {bone1Name=dyn_hair_05_lf, bone2Name=dyn_hair_04_lf, swing1LimitDegrees=9, swing2LimitDegrees=24, twistLimitDegrees=12}
 {bone1Name=dyn_hair_06_lf, bone2Name=dyn_hair_05_lf, swing1LimitDegrees=18, swing2LimitDegrees=24, twistLimitDegrees=12}
 {bone1Name=dyn_hair_07_lf, bone2Name=dyn_hair_06_lf, swing1LimitDegrees=18, swing2LimitDegrees=24, twistLimitDegrees=12}
 {
  bone1Name=base_hair_master
  bone2Name=head
  swing1LimitDegrees=45
  swing1Motion=Free
  swing2LimitDegrees=45
  swing2Motion=Free
  twistLimitDegrees=45
  twistMotion=Free
 }
 {
  bone1Name=dyn_hair_master
  bone2Name=base_hair_master
  swing1LimitDegrees=45
  swing1Motion=Free
  swing2LimitDegrees=45
  swing2Motion=Free
  twistLimitDegrees=45
  twistMotion=Free
 }
 {
  bone1Name=base_hair_rt
  bone2Name=head
  swing1LimitDegrees=45
  swing1Motion=Free
  swing2LimitDegrees=45
  swing2Motion=Free
  twistLimitDegrees=45
  twistMotion=Free
 }
 {bone1Name=dyn_hair_01_rt, bone2Name=base_hair_rt, swing1LimitDegrees=9, swing2LimitDegrees=10, twistLimitDegrees=12}
 {bone1Name=dyn_hair_02_rt, bone2Name=dyn_hair_01_rt, swing1LimitDegrees=9, swing2LimitDegrees=45, twistLimitDegrees=12}
 {bone1Name=dyn_hair_03_rt, bone2Name=dyn_hair_02_rt, swing1LimitDegrees=9, swing2LimitDegrees=45, twistLimitDegrees=12}
 {bone1Name=dyn_hair_04_rt, bone2Name=dyn_hair_03_rt, swing1LimitDegrees=9, swing2LimitDegrees=30, twistLimitDegrees=12}
 {bone1Name=dyn_hair_05_rt, bone2Name=dyn_hair_04_rt, swing1LimitDegrees=9, swing2LimitDegrees=24, twistLimitDegrees=12}
 {bone1Name=dyn_hair_06_rt, bone2Name=dyn_hair_05_rt, swing1LimitDegrees=18, swing2LimitDegrees=24, twistLimitDegrees=12}
 {bone1Name=dyn_hair_07_rt, bone2Name=dyn_hair_06_rt, swing1LimitDegrees=18, swing2LimitDegrees=24, twistLimitDegrees=12}
 {
  bone1Name=spine_04
  bone2Name=spine_03
  swing1LimitDegrees=45
  swing1Motion=Free
  swing2LimitDegrees=45
  swing2Motion=Free
  twistLimitDegrees=45
  twistMotion=Free
 }
 {
  bone1Name=spine_01
  bone2Name=pelvis
  swing1LimitDegrees=45
  swing1Motion=Free
  swing2LimitDegrees=45
  swing2Motion=Free
  twistLimitDegrees=45
  twistMotion=Free
 }
 {
  bone1Name=spine_02
  bone2Name=spine_01
  swing1LimitDegrees=45
  swing1Motion=Free
  swing2LimitDegrees=45
  swing2Motion=Free
  twistLimitDegrees=45
  twistMotion=Free
 }
 {
  bone1Name=spine_03
  bone2Name=spine_02
  swing1LimitDegrees=45
  swing1Motion=Free
  swing2LimitDegrees=45
  swing2Motion=Free
  twistLimitDegrees=45
  twistMotion=Free
 }
 {
  bone1Name=head
  bone2Name=neck_02
  swing1LimitDegrees=45
  swing1Motion=Free
  swing2LimitDegrees=45
  swing2Motion=Free
  twistLimitDegrees=45
  twistMotion=Free
 }
 {
  bone1Name=neck_01
  bone2Name=spine_05
  swing1LimitDegrees=45
  swing1Motion=Free
  swing2LimitDegrees=45
  swing2Motion=Free
  twistLimitDegrees=45
  twistMotion=Free
 }
 {
  bone1Name=upperarm_l
  bone2Name=clavicle_l
  swing1LimitDegrees=45
  swing1Motion=Free
  swing2LimitDegrees=45
  swing2Motion=Free
  twistLimitDegrees=45
  twistMotion=Free
 }
 {
  bone1Name=upperarm_r
  bone2Name=clavicle_r
  swing1LimitDegrees=45
  swing1Motion=Free
  swing2LimitDegrees=45
  swing2Motion=Free
  twistLimitDegrees=45
  twistMotion=Free
 }
 {
  bone1Name=dyn_hair_07_cr
  bone2Name=dyn_hair_07_lf
  swing1LimitDegrees=45
  swing1Motion=Free
  swing2LimitDegrees=45
  swing2Motion=Free
  twistLimitDegrees=45
  twistMotion=Free
 }
 {
  bone1Name=dyn_hair_07_cr
  bone2Name=dyn_hair_07_rt
  swing1LimitDegrees=45
  swing1Motion=Free
  swing2LimitDegrees=45
  swing2Motion=Free
  twistLimitDegrees=45
  twistMotion=Free
 }
 {
  bone1Name=dyn_hair_front_side_02_04_lf
  bone2Name=dyn_hair_front_side_02_03_lf
  swing1LimitDegrees=9
  swing2LimitDegrees=20
  twistLimitDegrees=0
 }
 {
  bone1Name=dyn_hair_front_side_02_05_lf
  bone2Name=dyn_hair_front_side_02_04_lf
  swing1LimitDegrees=9
  swing2LimitDegrees=20
  twistLimitDegrees=0
 }
 {
  bone1Name=dyn_hair_front_side_02_06_lf
  bone2Name=dyn_hair_front_side_02_05_lf
  swing1LimitDegrees=9
  swing2LimitDegrees=20
  twistLimitDegrees=0
 }
 {
  bone1Name=dyn_hair_front_side_02_07_lf
  bone2Name=dyn_hair_front_side_02_06_lf
  swing1LimitDegrees=9
  swing2LimitDegrees=20
  twistLimitDegrees=0
 }
 {
  bone1Name=dyn_hair_front_side_02_08_lf
  bone2Name=dyn_hair_front_side_02_07_lf
  swing1LimitDegrees=9
  swing2LimitDegrees=20
  twistLimitDegrees=0
 }
 {
  bone1Name=dyn_hair_front_side_02_04_rt
  bone2Name=dyn_hair_front_side_02_03_rt
  swing1LimitDegrees=9
  swing2LimitDegrees=20
  twistLimitDegrees=0
 }
 {
  bone1Name=dyn_hair_front_side_02_05_rt
  bone2Name=dyn_hair_front_side_02_04_rt
  swing1LimitDegrees=9
  swing2LimitDegrees=20
  twistLimitDegrees=0
 }
 {
  bone1Name=dyn_hair_front_side_02_06_rt
  bone2Name=dyn_hair_front_side_02_05_rt
  swing1LimitDegrees=9
  swing2LimitDegrees=20
  twistLimitDegrees=0
 }
 {
  bone1Name=dyn_hair_front_side_02_07_rt
  bone2Name=dyn_hair_front_side_02_06_rt
  swing1LimitDegrees=9
  swing2LimitDegrees=20
  twistLimitDegrees=0
 }
 {
  bone1Name=dyn_hair_front_side_02_08_rt
  bone2Name=dyn_hair_front_side_02_07_rt
  swing1LimitDegrees=9
  swing2LimitDegrees=20
  twistLimitDegrees=0
 }
 {bone1Name=dyn_hair_08_cr, bone2Name=dyn_hair_07_cr, swing1LimitDegrees=18, swing2LimitDegrees=24, twistLimitDegrees=12}
 {
  bone1Name=dyn_hair_front_side_02_03_lf
  bone2Name=base_hair_front_side_02_03_lf
  swing1LimitDegrees=9
  swing2LimitDegrees=20
  twistLimitDegrees=0
 }
 {
  bone1Name=dyn_hair_front_side_02_03_rt
  bone2Name=base_hair_front_side_02_03_rt
  swing1LimitDegrees=9
  swing2LimitDegrees=20
  twistLimitDegrees=0
 }
 {
  bone1Name=base_hair_front_side_02_03_lf
  bone2Name=head
  swing1LimitDegrees=45
  swing1Motion=Free
  swing2LimitDegrees=45
  swing2Motion=Free
  twistLimitDegrees=45
  twistMotion=Free
 }
 {
  bone1Name=base_hair_front_side_02_03_rt
  bone2Name=head
  swing1LimitDegrees=45
  swing1Motion=Free
  swing2LimitDegrees=45
  swing2Motion=Free
  twistLimitDegrees=45
  twistMotion=Free
 }
 {
  bone1Name=dyn_pauldron_r
  bone2Name=upperarm_r
  swing1LimitDegrees=45
  swing1Motion=Free
  swing2LimitDegrees=45
  swing2Motion=Free
  twistLimitDegrees=45
  twistMotion=Free
 }
 {
  bone1Name=dyn_pauldron_l
  bone2Name=upperarm_l
  swing1LimitDegrees=45
  swing1Motion=Free
  swing2LimitDegrees=45
  swing2Motion=Free
  twistLimitDegrees=45
  twistMotion=Free
 }
 {
  bone1Name=thorax
  bone2Name=spine_04
  swing1LimitDegrees=45
  swing1Motion=Free
  swing2LimitDegrees=45
  swing2Motion=Free
  twistLimitDegrees=45
  twistMotion=Free
 }
 {
  bone1Name=dyn_pauldron_fr_l
  bone2Name=upperarm_l
  swing1LimitDegrees=45
  swing1Motion=Free
  swing2LimitDegrees=45
  swing2Motion=Free
  twistLimitDegrees=45
  twistMotion=Free
 }
 {
  bone1Name=dyn_pauldron_fr_r
  bone2Name=upperarm_r
  swing1LimitDegrees=45
  swing1Motion=Free
  swing2LimitDegrees=45
  swing2Motion=Free
  twistLimitDegrees=45
  twistMotion=Free
 }
]
