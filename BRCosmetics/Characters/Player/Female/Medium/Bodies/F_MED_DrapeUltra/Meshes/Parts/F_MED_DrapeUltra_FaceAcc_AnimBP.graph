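# Post-physics hair AnimBP for the DrapeUltra female FaceAcc mesh, copying the attached parent's pose through a hair-spline Control Rig.
# A backbling check selects between two RigidBody sims with separate physics assets; the backpack branch also squashes, offsets and tilts hair bones.
# AnimDynamics then springs bangs, which pitch up when falling without skydiving, before fixed base-hair offsets.
# A parameter-driven function updates RigidBody gravity each frame.
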AnimBlueprint /BRCosmetics/Characters/Player/Female/Medium/Bodies/F_MED_DrapeUltra/Meshes/Parts/F_MED_DrapeUltra_FaceAcc_AnimBP
target_skeleton=&/F_MED_DrapeUltra_FaceAcc_Skeleton
graphs [
 AnimGraph {nodes [
   n1 = RigidBody!(
    ComponentPose=Local To Component(LocalPose=Use cached pose 'Hair_Spline_Ctrl_Rig'())
    Alpha
    OverrideWorldGravity:Vector=Get GravityOverride()
    ExternalForce
   ) -> (Pose) {
    bClampLinearTranslationLimitToRefPose=true
    bOverrideWorldGravity=true
    overridePhysicsAsset=&/F_MED_DrapeUltra_FaceAcc_USES_Backpack_Physics
    simulationSpace=BaseBoneSpace
    alphaBoolBlend {blendOption=Linear}
    baseBoneRef {boneName=spine_04}
    overrideWorldGravity {z=-980}
    simSpaceSettings {worldAlpha=0.4, externalAngularVelocity {x=-0.5}, externalLinearDragV {x=-0.5}, externalLinearVelocity {x=-0.5}}
   }
   n2 = "Transform (Modify) Bone - Bone: scl_hair_03_lf"(ComponentPose=@n1, Alpha=1, Translation="0,0,0", Rotation="0,0,0", Scale="1,1,0.75") {
    rotationMode=BMM_Additive
    rotationSpace=BCS_BoneSpace
    scaleMode=BMM_Replace
    translationMode=BMM_Additive
    translationSpace=BCS_BoneSpace
    boneToModify {boneName=scl_hair_03_lf}
    scale {z=0.75}
   }
   n3 = "Transform (Modify) Bone - Bone: scl_hair_03_cr"(ComponentPose=@n2, Alpha=1, Translation="0,0,0", Rotation="0,0,0", Scale="1,1,0.75") {
    rotationMode=BMM_Additive
    rotationSpace=BCS_BoneSpace
    scaleMode=BMM_Replace
    translationMode=BMM_Additive
    translationSpace=BCS_BoneSpace
    boneToModify {boneName=scl_hair_03_cr}
    scale {z=0.75}
   }
   n4 = "Transform (Modify) Bone - Bone: scl_hair_03_rt"(ComponentPose=@n3, Alpha=1, Translation="0,0,0", Rotation="0,0,0", Scale="1,1,0.75") {
    rotationMode=BMM_Additive
    rotationSpace=BCS_BoneSpace
    scaleMode=BMM_Replace
    translationMode=BMM_Additive
    translationSpace=BCS_BoneSpace
    boneToModify {boneName=scl_hair_03_rt}
    scale {z=0.75}
   }
   n5 = "Transform (Modify) Bone - Bone: scl_hair_04_lf"(ComponentPose=@n4, Alpha=1, Translation="0,0,-1", Rotation="0,0,0", Scale="1,1,0.75") {
    rotationMode=BMM_Additive
    rotationSpace=BCS_BoneSpace
    scaleMode=BMM_Replace
    translationMode=BMM_Additive
    translationSpace=BCS_BoneSpace
    boneToModify {boneName=scl_hair_04_lf}
    scale {z=0.75}
    translation {z=-1}
   }
   n6 = "Transform (Modify) Bone - Bone: scl_hair_04_cr"(ComponentPose=@n5, Alpha=1, Translation="0,0,-1", Rotation="0,0,0", Scale="1,1,0.75") {
    rotationMode=BMM_Additive
    rotationSpace=BCS_BoneSpace
    scaleMode=BMM_Replace
    translationMode=BMM_Additive
    translationSpace=BCS_BoneSpace
    boneToModify {boneName=scl_hair_04_cr}
    scale {z=0.75}
    translation {z=-1}
   }
   n7 = "Transform (Modify) Bone - Bone: scl_hair_04_rt"(ComponentPose=@n6, Alpha=1, Translation="0,0,-1", Rotation="0,0,0", Scale="1,1,0.75") {
    rotationMode=BMM_Additive
    rotationSpace=BCS_BoneSpace
    scaleMode=BMM_Replace
    translationMode=BMM_Additive
    translationSpace=BCS_BoneSpace
    boneToModify {boneName=scl_hair_04_rt}
    scale {z=0.75}
    translation {z=-1}
   }
   n8 = "Transform (Modify) Bone - Bone: scl_hair_05_lf"(ComponentPose=@n7, Alpha=1, Translation="0,0,-0.5", Rotation="0,0,0", Scale="1,1,0.75") {
    rotationMode=BMM_Additive
    rotationSpace=BCS_BoneSpace
    scaleMode=BMM_Replace
    translationMode=BMM_Additive
    translationSpace=BCS_BoneSpace
    boneToModify {boneName=scl_hair_05_lf}
    scale {z=0.75}
    translation {z=-0.5}
   }
   n9 = "Transform (Modify) Bone - Bone: scl_hair_05_cr"(ComponentPose=@n8, Alpha=1, Translation="0,0,-0.5", Rotation="0,0,0", Scale="1,1,0.75") {
    rotationMode=BMM_Additive
    rotationSpace=BCS_BoneSpace
    scaleMode=BMM_Replace
    translationMode=BMM_Additive
    translationSpace=BCS_BoneSpace
    boneToModify {boneName=scl_hair_05_cr}
    scale {z=0.75}
    translation {z=-0.5}
   }
   n10 = "Transform (Modify) Bone - Bone: scl_hair_05_rt"(ComponentPose=@n9, Alpha=1, Translation="0,0,-0.5", Rotation="0,0,0", Scale="1,1,0.75") {
    rotationMode=BMM_Additive
    rotationSpace=BCS_BoneSpace
    scaleMode=BMM_Replace
    translationMode=BMM_Additive
    translationSpace=BCS_BoneSpace
    boneToModify {boneName=scl_hair_05_rt}
    scale {z=0.75}
    translation {z=-0.5}
   }
   n11 = "Transform (Modify) Bone - Bone: def_hair_03_lf"(ComponentPose=@n10, Alpha=1, Translation="0,0,-2.5", Rotation="0,0,0", Scale="1,1,0.75") {
    rotationMode=BMM_Additive
    rotationSpace=BCS_BoneSpace
    scaleMode=BMM_Replace
    translationMode=BMM_Additive
    translationSpace=BCS_BoneSpace
    boneToModify {boneName=def_hair_03_lf}
    scale {z=0.75}
    translation {z=-2.5}
   }
   n12 = "Transform (Modify) Bone - Bone: def_hair_03_rt"(ComponentPose=@n11, Alpha=1, Translation="0,0,-2.5", Rotation="0,0,0", Scale="1,1,0.75") {
    rotationMode=BMM_Additive
    rotationSpace=BCS_BoneSpace
    scaleMode=BMM_Replace
    translationMode=BMM_Additive
    translationSpace=BCS_BoneSpace
    boneToModify {boneName=def_hair_03_rt}
    scale {z=0.75}
    translation {z=-2.5}
   }
   n13 = "Transform (Modify) Bone - Bone: dyn_hair_05_cr"(ComponentPose=@n12, Alpha=1, Translation="0,0,0", Rotation="-5,0,0", Scale="1,1,1") {
    rotationMode=BMM_Additive
    rotationSpace=BCS_BoneSpace
    translationMode=BMM_Additive
    translationSpace=BCS_BoneSpace
    boneToModify {boneName=dyn_hair_05_cr}
    rotation {pitch=-5}
   }
   n14 = "Transform (Modify) Bone - Bone: dyn_hair_05_lf"(ComponentPose=@n13, Alpha=1, Translation="0,0,0", Rotation="-5,0,0", Scale="1,1,1") {
    rotationMode=BMM_Additive
    rotationSpace=BCS_BoneSpace
    translationMode=BMM_Additive
    translationSpace=BCS_BoneSpace
    boneToModify {boneName=dyn_hair_05_lf}
    rotation {pitch=-5}
   }
   n15 = "Transform (Modify) Bone - Bone: dyn_hair_05_rt"(ComponentPose=@n14, Alpha=1, Translation="0,0,0", Rotation="-5,0,0", Scale="1,1,1") {
    rotationMode=BMM_Additive
    rotationSpace=BCS_BoneSpace
    translationMode=BMM_Additive
    translationSpace=BCS_BoneSpace
    boneToModify {boneName=dyn_hair_05_rt}
    rotation {pitch=-5}
   }
   n16 = "Transform (Modify) Bone - Bone: dyn_hair_06_cr"(ComponentPose=@n15, Alpha=1, Translation="0,0,0", Rotation="-5,0,0", Scale="1,1,1") {
    rotationMode=BMM_Additive
    rotationSpace=BCS_BoneSpace
    translationMode=BMM_Additive
    translationSpace=BCS_BoneSpace
    boneToModify {boneName=dyn_hair_06_cr}
    rotation {pitch=-5}
   }
   n17 = "Transform (Modify) Bone - Bone: dyn_hair_06_lf"(ComponentPose=@n16, Alpha=1, Translation="0,0,0", Rotation="-5,0,0", Scale="1,1,1") {
    rotationMode=BMM_Additive
    rotationSpace=BCS_BoneSpace
    translationMode=BMM_Additive
    translationSpace=BCS_BoneSpace
    boneToModify {boneName=dyn_hair_06_lf}
    rotation {pitch=-5}
   }
   n18 = "Transform (Modify) Bone - Bone: dyn_hair_06_rt"(ComponentPose=@n17, Alpha=1, Translation="0,0,0", Rotation="-5,0,0", Scale="1,1,1") {
    rotationMode=BMM_Additive
    rotationSpace=BCS_BoneSpace
    translationMode=BMM_Additive
    translationSpace=BCS_BoneSpace
    boneToModify {boneName=dyn_hair_06_rt}
    rotation {pitch=-5}
   }
   n19 = Component To Local(ComponentPose=@n18)
   n20 = Use cached pose 'No_Backpack'()
   n21 = Get Is Wearing Backbling()
   n22 = Anim Dynamics - Bone: dyn_bangs_01_c!(
    ComponentPose=Local To Component(LocalPose=Blend Poses by bool(BlendPose_0=@n19, BlendPose_1=@n20, bActiveValue=@n21))
    Alpha=0.333
    GravityOverride:Vector="0,0,-980"
    ExternalForce="0,0,0"
   ) -> (Pose) {
    angularDampingOverride=1
    angularSpringConstant=12000
    bAngularSpring=true
    bOverrideAngularDamping=true
    bUseGravityOverride=true
    gravityScale=0.35
    simulationSpace=BoneRelative
    alphaBoolBlend {blendOption=Linear}
    boundBone {boneName=dyn_bangs_01_c}
    componentLinearAccScale=<1 1 1>
    componentLinearVelScale=<1 1 1>
    gravityOverride {z=-980}
    physicsBodyDefinitions [{
      boundBone {boneName=dyn_bangs_01_c}
      boxExtents=<25 25 25>
      constraintSetup {angularLimitsMax=<0 0 60>, angularLimitsMin=<0 -60 -60>, angularTarget=<1 0 0>}
      localJointOffset=<2 0 0>
     }]
    relativeSpaceBone {boneName=root}
   }
   n23 = Anim Dynamics - Bone: dyn_bangs_02_c!(ComponentPose=@n22, Alpha=0.333, GravityOverride:Vector="0,0,-980", ExternalForce="0,0,0") -> (Pose) {
    angularDampingOverride=1
    angularSpringConstant=12000
    bAngularSpring=true
    bOverrideAngularDamping=true
    bUseGravityOverride=true
    gravityScale=0.35
    simulationSpace=BoneRelative
    alphaBoolBlend {blendOption=Linear}
    boundBone {boneName=dyn_bangs_02_c}
    componentLinearAccScale=<1 1 1>
    componentLinearVelScale=<1 1 1>
    gravityOverride {z=-980}
    physicsBodyDefinitions [{
      boundBone {boneName=dyn_bangs_02_c}
      boxExtents=<25 25 25>
      constraintSetup {angularLimitsMax=<0 0 60>, angularLimitsMin=<0 -60 -60>, angularTarget=<1 0 0>}
      localJointOffset=<2 0 0>
     }]
    relativeSpaceBone {boneName=root}
   }
   n24 = Anim Dynamics - Bone: dyn_bangs_03_c!(ComponentPose=@n23, Alpha=0.333, GravityOverride:Vector="0,0,-980", ExternalForce="0,0,0") -> (Pose) {
    angularDampingOverride=1
    angularSpringConstant=12000
    bAngularSpring=true
    bOverrideAngularDamping=true
    bUseGravityOverride=true
    gravityScale=0.35
    simulationSpace=BoneRelative
    alphaBoolBlend {blendOption=Linear}
    boundBone {boneName=dyn_bangs_03_c}
    componentLinearAccScale=<1 1 1>
    componentLinearVelScale=<1 1 1>
    gravityOverride {z=-980}
    physicsBodyDefinitions [{
      boundBone {boneName=dyn_bangs_03_c}
      boxExtents=<25 25 25>
      constraintSetup {angularLimitsMax=<0 0 60>, angularLimitsMin=<0 -60 -60>, angularTarget=<1 0 0>}
      localJointOffset=<2 0 0>
     }]
    relativeSpaceBone {boneName=root}
   }
   n25 = Anim Dynamics - Bone: dyn_bangs_01_l!(ComponentPose=@n24, Alpha=0.333, GravityOverride:Vector="0,0,-980", ExternalForce="0,0,0") -> (Pose) {
    angularDampingOverride=1
    angularSpringConstant=12000
    bAngularSpring=true
    bOverrideAngularDamping=true
    bUseGravityOverride=true
    gravityScale=0.35
    simulationSpace=BoneRelative
    alphaBoolBlend {blendOption=Linear}
    boundBone {boneName=dyn_bangs_01_l}
    componentLinearAccScale=<1 1 1>
    componentLinearVelScale=<1 1 1>
    gravityOverride {z=-980}
    physicsBodyDefinitions [{
      boundBone {boneName=dyn_bangs_01_l}
      boxExtents=<25 25 25>
      constraintSetup {angularLimitsMax=<0 0 60>, angularLimitsMin=<0 -60 -60>, angularTarget=<1 0 0>}
      localJointOffset=<2 0 0>
     }]
    relativeSpaceBone {boneName=root}
   }
   n26 = Anim Dynamics - Bone: dyn_bangs_02_l!(ComponentPose=@n25, Alpha=0.333, GravityOverride:Vector="0,0,-980", ExternalForce="0,0,0") -> (Pose) {
    angularDampingOverride=1
    angularSpringConstant=12000
    bAngularSpring=true
    bOverrideAngularDamping=true
    bUseGravityOverride=true
    gravityScale=0.35
    simulationSpace=BoneRelative
    alphaBoolBlend {blendOption=Linear}
    boundBone {boneName=dyn_bangs_02_l}
    componentLinearAccScale=<1 1 1>
    componentLinearVelScale=<1 1 1>
    gravityOverride {z=-980}
    physicsBodyDefinitions [{
      boundBone {boneName=dyn_bangs_02_l}
      boxExtents=<25 25 25>
      constraintSetup {angularLimitsMax=<0 0 60>, angularLimitsMin=<0 -60 -60>, angularTarget=<1 0 0>}
      localJointOffset=<2 0 0>
     }]
    relativeSpaceBone {boneName=root}
   }
   n27 = Anim Dynamics - Bone: dyn_bangs_03_l!(ComponentPose=@n26, Alpha=0.333, GravityOverride:Vector="0,0,-980", ExternalForce="0,0,0") -> (Pose) {
    angularDampingOverride=1
    angularSpringConstant=12000
    bAngularSpring=true
    bOverrideAngularDamping=true
    bUseGravityOverride=true
    gravityScale=0.35
    simulationSpace=BoneRelative
    alphaBoolBlend {blendOption=Linear}
    boundBone {boneName=dyn_bangs_03_l}
    componentLinearAccScale=<1 1 1>
    componentLinearVelScale=<1 1 1>
    gravityOverride {z=-980}
    physicsBodyDefinitions [{
      boundBone {boneName=dyn_bangs_03_l}
      boxExtents=<25 25 25>
      constraintSetup {angularLimitsMax=<0 0 60>, angularLimitsMin=<0 -60 -60>, angularTarget=<1 0 0>}
      localJointOffset=<2 0 0>
     }]
    relativeSpaceBone {boneName=root}
   }
   n28 = Anim Dynamics - Bone: dyn_bangs_01_r!(ComponentPose=@n27, Alpha=0.333, GravityOverride:Vector="0,0,-980", ExternalForce="0,0,0") -> (Pose) {
    angularDampingOverride=1
    angularSpringConstant=12000
    bAngularSpring=true
    bOverrideAngularDamping=true
    bUseGravityOverride=true
    gravityScale=0.35
    simulationSpace=BoneRelative
    alphaBoolBlend {blendOption=Linear}
    boundBone {boneName=dyn_bangs_01_r}
    componentLinearAccScale=<1 1 1>
    componentLinearVelScale=<1 1 1>
    gravityOverride {z=-980}
    physicsBodyDefinitions [{
      boundBone {boneName=dyn_bangs_01_r}
      boxExtents=<25 25 25>
      constraintSetup {angularLimitsMax=<0 0 60>, angularLimitsMin=<0 -60 -60>, angularTarget=<1 0 0>}
      localJointOffset=<2 0 0>
     }]
    relativeSpaceBone {boneName=root}
   }
   n29 = Anim Dynamics - Bone: dyn_bangs_02_r!(ComponentPose=@n28, Alpha=0.333, GravityOverride:Vector="0,0,-980", ExternalForce="0,0,0") -> (Pose) {
    angularDampingOverride=1
    angularSpringConstant=12000
    bAngularSpring=true
    bOverrideAngularDamping=true
    bUseGravityOverride=true
    gravityScale=0.35
    simulationSpace=BoneRelative
    alphaBoolBlend {blendOption=Linear}
    boundBone {boneName=dyn_bangs_02_r}
    componentLinearAccScale=<1 1 1>
    componentLinearVelScale=<1 1 1>
    gravityOverride {z=-980}
    physicsBodyDefinitions [{
      boundBone {boneName=dyn_bangs_02_r}
      boxExtents=<25 25 25>
      constraintSetup {angularLimitsMax=<0 0 60>, angularLimitsMin=<0 -60 -60>, angularTarget=<1 0 0>}
      localJointOffset=<2 0 0>
     }]
    relativeSpaceBone {boneName=root}
   }
   n30 = Anim Dynamics - Bone: dyn_bangs_03_r!(ComponentPose=@n29, Alpha=0.333, GravityOverride:Vector="0,0,-980", ExternalForce="0,0,0") -> (Pose) {
    angularDampingOverride=1
    angularSpringConstant=12000
    bAngularSpring=true
    bOverrideAngularDamping=true
    bUseGravityOverride=true
    gravityScale=0.35
    simulationSpace=BoneRelative
    alphaBoolBlend {blendOption=Linear}
    boundBone {boneName=dyn_bangs_03_r}
    componentLinearAccScale=<1 1 1>
    componentLinearVelScale=<1 1 1>
    gravityOverride {z=-980}
    physicsBodyDefinitions [{
      boundBone {boneName=dyn_bangs_03_r}
      boxExtents=<25 25 25>
      constraintSetup {angularLimitsMax=<0 0 60>, angularLimitsMin=<0 -60 -60>, angularTarget=<1 0 0>}
      localJointOffset=<2 0 0>
     }]
    relativeSpaceBone {boneName=root}
   }
   n31 = Get isFallingNotSkydiving() -> (isFallingNotSkydiving:bool)
   n32 = "Transform (Modify) Bone - Bone: dyn_bangs_01_c"!(ComponentPose=@n30, bAlphaBoolEnabled:bool=@n31, Translation="0,0,0", Rotation="90,0,0", Scale="1,1,1") -> (Pose) {
    alphaInputType=Bool
    rotationMode=BMM_Additive
    rotationSpace=BCS_BoneSpace
    translationMode=BMM_Additive
    translationSpace=BCS_BoneSpace
    alphaBoolBlend {blendInTime=1.25, blendOption=Linear, blendOutTime=0.1}
    boneToModify {boneName=dyn_bangs_01_c}
    rotation {pitch=90}
   }
   n33 = "Transform (Modify) Bone - Bone: dyn_bangs_02_c"!(ComponentPose=@n32, bAlphaBoolEnabled:bool=@n31, Translation="0,0,0", Rotation="90,0,0", Scale="1,1,1") -> (Pose) {
    alphaInputType=Bool
    rotationMode=BMM_Additive
    rotationSpace=BCS_BoneSpace
    translationMode=BMM_Additive
    translationSpace=BCS_BoneSpace
    alphaBoolBlend {blendInTime=1.25, blendOption=Linear, blendOutTime=0.1}
    boneToModify {boneName=dyn_bangs_02_c}
    rotation {pitch=90}
   }
   n34 = "Transform (Modify) Bone - Bone: dyn_bangs_03_c"!(ComponentPose=@n33, bAlphaBoolEnabled:bool=@n31, Translation="0,0,0", Rotation="90,0,0", Scale="1,1,1") -> (Pose) {
    alphaInputType=Bool
    rotationMode=BMM_Additive
    rotationSpace=BCS_BoneSpace
    translationMode=BMM_Additive
    translationSpace=BCS_BoneSpace
    alphaBoolBlend {blendInTime=1.25, blendOption=Linear, blendOutTime=0.1}
    boneToModify {boneName=dyn_bangs_03_c}
    rotation {pitch=90}
   }
   n35 = "Transform (Modify) Bone - Bone: dyn_bangs_01_l"!(ComponentPose=@n34, bAlphaBoolEnabled:bool=@n31, Translation="0,0,0", Rotation="90,0,0", Scale="1,1,1") -> (Pose) {
    alphaInputType=Bool
    rotationMode=BMM_Additive
    rotationSpace=BCS_BoneSpace
    translationMode=BMM_Additive
    translationSpace=BCS_BoneSpace
    alphaBoolBlend {blendInTime=1.25, blendOption=Linear, blendOutTime=0.1}
    boneToModify {boneName=dyn_bangs_01_l}
    rotation {pitch=90}
   }
   n36 = "Transform (Modify) Bone - Bone: dyn_bangs_02_l"!(ComponentPose=@n35, bAlphaBoolEnabled:bool=@n31, Translation="0,0,0", Rotation="90,0,0", Scale="1,1,1") -> (Pose) {
    alphaInputType=Bool
    rotationMode=BMM_Additive
    rotationSpace=BCS_BoneSpace
    translationMode=BMM_Additive
    translationSpace=BCS_BoneSpace
    alphaBoolBlend {blendInTime=1.25, blendOption=Linear, blendOutTime=0.1}
    boneToModify {boneName=dyn_bangs_02_l}
    rotation {pitch=90}
   }
   n37 = "Transform (Modify) Bone - Bone: dyn_bangs_03_l"!(ComponentPose=@n36, bAlphaBoolEnabled:bool=@n31, Translation="0,0,0", Rotation="90,0,0", Scale="1,1,1") -> (Pose) {
    alphaInputType=Bool
    rotationMode=BMM_Additive
    rotationSpace=BCS_BoneSpace
    translationMode=BMM_Additive
    translationSpace=BCS_BoneSpace
    alphaBoolBlend {blendInTime=1.25, blendOption=Linear, blendOutTime=0.1}
    boneToModify {boneName=dyn_bangs_03_l}
    rotation {pitch=90}
   }
   n38 = "Transform (Modify) Bone - Bone: dyn_bangs_01_r"!(ComponentPose=@n37, bAlphaBoolEnabled:bool=@n31, Translation="0,0,0", Rotation="90,0,0", Scale="1,1,1") -> (Pose) {
    alphaInputType=Bool
    rotationMode=BMM_Additive
    rotationSpace=BCS_BoneSpace
    translationMode=BMM_Additive
    translationSpace=BCS_BoneSpace
    alphaBoolBlend {blendInTime=1.25, blendOption=Linear, blendOutTime=0.1}
    boneToModify {boneName=dyn_bangs_01_r}
    rotation {pitch=90}
   }
   n39 = "Transform (Modify) Bone - Bone: dyn_bangs_02_r"!(ComponentPose=@n38, bAlphaBoolEnabled:bool=@n31, Translation="0,0,0", Rotation="90,0,0", Scale="1,1,1") -> (Pose) {
    alphaInputType=Bool
    rotationMode=BMM_Additive
    rotationSpace=BCS_BoneSpace
    translationMode=BMM_Additive
    translationSpace=BCS_BoneSpace
    alphaBoolBlend {blendInTime=1.25, blendOption=Linear, blendOutTime=0.1}
    boneToModify {boneName=dyn_bangs_02_r}
    rotation {pitch=90}
   }
   n40 = "Transform (Modify) Bone - Bone: dyn_bangs_03_r"!(ComponentPose=@n39, bAlphaBoolEnabled:bool=@n31, Translation="0,0,0", Rotation="90,0,0", Scale="1,1,1") -> (Pose) {
    alphaInputType=Bool
    rotationMode=BMM_Additive
    rotationSpace=BCS_BoneSpace
    translationMode=BMM_Additive
    translationSpace=BCS_BoneSpace
    alphaBoolBlend {blendInTime=1.25, blendOption=Linear, blendOutTime=0.1}
    boneToModify {boneName=dyn_bangs_03_r}
    rotation {pitch=90}
   }
   n41 = "Transform (Modify) Bone - Bone: base_hair_cr"(ComponentPose=@n40, Alpha=1, Translation="-2,0,1", Rotation="0,0,0", Scale="1,1,1") {
    rotationMode=BMM_Additive
    rotationSpace=BCS_BoneSpace
    scaleMode=BMM_Replace
    translationMode=BMM_Additive
    translationSpace=BCS_BoneSpace
    boneToModify {boneName=base_hair_cr}
    translation {x=-2, z=1}
   }
   n42 = "Transform (Modify) Bone - Bone: base_hair_lf"(ComponentPose=@n41, Alpha=1, Translation="-2,0,1", Rotation="0,0,0", Scale="1,1,1") {
    rotationMode=BMM_Additive
    rotationSpace=BCS_BoneSpace
    scaleMode=BMM_Replace
    translationMode=BMM_Additive
    translationSpace=BCS_BoneSpace
    boneToModify {boneName=base_hair_lf}
    translation {x=-2, z=1}
   }
   n43 = "Transform (Modify) Bone - Bone: base_hair_rt"(ComponentPose=@n42, Alpha=1, Translation="-2,0,1", Rotation="0,0,0", Scale="1,1,1") {
    rotationMode=BMM_Additive
    rotationSpace=BCS_BoneSpace
    scaleMode=BMM_Replace
    translationMode=BMM_Additive
    translationSpace=BCS_BoneSpace
    boneToModify {boneName=base_hair_rt}
    translation {x=-2, z=1}
   }
   Output Pose(Result=Component To Local(ComponentPose=@n43)) {layerGroup=None}
   n44 = Control Rig(Source=Copy Pose From Mesh() {bUseAttachedParent=true}) {controlRigAssetReference {blueprintRigClass=&/F_MED_DrapeUltraLime_Hair_Spl_CtrlRig.F_MED_DrapeUltraLime_Hair_Spl_CtrlRig_C}}
   Save cached pose 'Hair_Spline_Ctrl_Rig'(Pose=@n44) {cacheName=Hair_Spline_Ctrl_Rig}
   n45 = RigidBody(ComponentPose=Local To Component(LocalPose=Use cached pose 'Hair_Spline_Ctrl_Rig'()), OverrideWorldGravity=Get GravityOverride()) {
    bClampLinearTranslationLimitToRefPose=true
    bOverrideWorldGravity=true
    overridePhysicsAsset=&/F_MED_DrapeUltra_FaceAcc_WITHOUT_Backpack_Physics
    simulationSpace=BaseBoneSpace
    alphaBoolBlend {blendOption=Linear}
    baseBoneRef {boneName=spine_04}
    overrideWorldGravity {z=-980}
    simSpaceSettings {worldAlpha=0.4, externalAngularVelocity {x=-0.5}, externalLinearDragV {x=-0.5}, externalLinearVelocity {x=-0.5}}
   }
   Save cached pose 'No_Backpack'(Pose=Component To Local(ComponentPose=@n45)) {cacheName=No_Backpack}
  ]}
 EventGraph {nodes [
   K2Node_Event_0 = CustomEvent!() -> (OutputDelegate, then, DeltaTimeX:float=~) {is=title K2Node_CustomEvent}
   > Sequence()
   n1 = Get bIsSkyDiving()
   > n2 = Branch(Condition=AND Boolean(A=Get bIsFalling(), B=NOT Boolean(A=@n1)))
   > Set isFallingNotSkydiving(isFallingNotSkydiving:bool=true) -> (Output_Get:bool)
   Set isFallingNotSkydiving(execute=@n2.else, isFallingNotSkydiving=false)
   Event BlueprintUpdateAnimation()
   > n3 = "Gravity Override State Function (Parameter Asset) (Nativized)"(
    CharacterPartAnimInstance=Self-Reference()
    GameTimeInSeconds=GetGameTimeInSeconds()
    DebugJointName=neck_02
    Parameters=&/F_MED_DrapeUltra_Parameters
   )
   > Set GravityOverride(GravityOverride=@n3)
   > Set Skyfall No Backpack(Skyfall No Backpack:bool=AND Boolean(A=Get bIsFalling(), B=Get bIsSkyDiving())) -> (Output_Get:bool)
   Event OnCharacterCustomizationLoaded()
   n4 = CheckForBackbling()
   > Set Is Wearing Backbling(Is Wearing Backbling=@n4)
   > Set Has Backbling(Has Backbling:double="To Float (Boolean)"(InBool=@n4)) -> (Output_Get:double)
   > Set No Backbling(No Backbling:double="To Float (Boolean)"(InBool=InvertBool!(bValue:bool=@n4) -> (ReturnValue:bool))) -> (Output_Get:double)
   Get Gravity Override Params() -> (Gravity Override Params:Fort Gravity Override Parameter Set Structure)
  ]}
]
variables [
 Gravity Override Params {
  type=FortGravityOverrideParameterSet
  default=(Skydive_GoingUp=(GravityOverride=()),Skydive_FloatingDown=(GravityOverride=()),Skydive_DivingDown=(GravityOverride=()),Skydive_Parachuting=(GravityOverride=()),OnGround_Standing=(GravityOverride=()),OnGround_Moving=(GravityOverride=()),OnGround_Moving_HighRotationRate=(GravityOverride=()),OnGround_Crouch_Moving=(GravityOverride=()),EmoteOrMelee=(GravityOverride=()),Swimming=(GravityOverride=()),Front_End=(GravityOverride=()),DBNO=(GravityOverride=()),DBNO_Carried=(GravityOverride=()),OnGround_Falling=(GravityOverride=()),Ziplining=(GravityOverride=()),Skydive_Surfing=(GravityOverride=()),Riding_Standing=(GravityOverride=()),Riding_Moving=(GravityOverride=()),Riding_Moving_Targeting=(GravityOverride=()),Tactical_Sprint=(GravityOverride=()),Rail_Grinding=(GravityOverride=()),)
 }
 GravityOverride {type=Vector}
 Has Backbling {type=number}
 Is Wearing Backbling {type=boolean}
 isFallingNotSkydiving {type=boolean}
 No Backbling {type=number}
 RBAN Settings Out {type=FortRigidBodyAnimNodeParameters}
 Skyfall No Backpack {type=boolean}
]
