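# Physics asset for the DrapeUltra female medium face accessory, apparently providing collision between its dynamic hair bones and the cape, going by its name.
# Three chains (center, left, right) each carry one capsule on bones 02 through 05, largest at the tip.
# Constraints link neighboring bodies within each chain, with no links across chains.
# Swing and twist motions are Free, so the stored angular limits do not restrict rotation.
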PhysicsAsset /BRCosmetics/Characters/Player/Female/Medium/Bodies/F_MED_DrapeUltra/Meshes/Parts/F_MED_DrapeUltra_FaceAcc_CapeCollide
bodies {
 dyn_hair_02_cr [{shapeType=Capsule, length=3.6326, radius=3.8959, shapeName=dyn_hair_04_cr_capsule, center=<3.4024 0 0>, rotation=<rot 90 0 0>}]
 dyn_hair_02_lf [{shapeType=Capsule, length=3.6326, radius=3.8959, shapeName=dyn_hair_04_cr_capsule, center=<3.4024 0 0>, rotation=<rot 90 0 0>}]
 dyn_hair_02_rt [{shapeType=Capsule, length=3.6326, radius=3.8959, shapeName=dyn_hair_04_cr_capsule, center=<3.4024 0 0>, rotation=<rot 90 0 0>}]
 dyn_hair_03_cr [{shapeType=Capsule, length=3.6326, radius=3.8959, shapeName=dyn_hair_04_cr_capsule, center=<3.4024 0 0>, rotation=<rot 90 0 0>}]
 dyn_hair_03_lf [{shapeType=Capsule, length=3.6326, radius=3.8959, shapeName=dyn_hair_04_cr_capsule, center=<3.4024 0 0>, rotation=<rot 90 0 0>}]
 dyn_hair_03_rt [{shapeType=Capsule, length=3.6326, radius=3.8959, shapeName=dyn_hair_04_cr_capsule, center=<3.4024 0 0>, rotation=<rot 90 0 0>}]
 dyn_hair_04_cr [{shapeType=Capsule, length=4.789, radius=3.7113, shapeName=dyn_hair_04_cr_capsule, center=<3.8165 0 0>, rotation=<rot 80.0001 -180 -180>}]
 dyn_hair_04_lf [{shapeType=Capsule, length=4.789, radius=3.7113, shapeName=dyn_hair_04_cr_capsule, center=<3.8165 0 0>, rotation=<rot 80.0001 -180 -180>}]
 dyn_hair_04_rt [{shapeType=Capsule, length=4.789, radius=3.7113, shapeName=dyn_hair_04_cr_capsule, center=<3.8165 0 0>, rotation=<rot 80.0001 -180 -180>}]
 dyn_hair_05_cr [{shapeType=Capsule, length=8.245, radius=4.173, shapeName=dyn_hair_04_cr_capsule, center=<5.0317 0 0>, rotation=<rot 90 0 0>}]
 dyn_hair_05_lf [{shapeType=Capsule, length=8.245, radius=4.173, shapeName=dyn_hair_04_cr_capsule, center=<5.0317 0 0>, rotation=<rot 90 0 0>}]
 dyn_hair_05_rt [{shapeType=Capsule, length=8.245, radius=4.173, shapeName=dyn_hair_04_cr_capsule, center=<5.0317 0 0>, rotation=<rot 90 0 0>}]
}
constraints [
 {
  bone1Name=dyn_hair_05_cr
  bone2Name=dyn_hair_04_cr
  swing1LimitDegrees=45
  swing1Motion=Free
  swing2LimitDegrees=45
  swing2Motion=Free
  twistLimitDegrees=45
  twistMotion=Free
 }
 {
  bone1Name=dyn_hair_04_cr
  bone2Name=dyn_hair_03_cr
  swing1LimitDegrees=45
  swing1Motion=Free
  swing2LimitDegrees=45
  swing2Motion=Free
  twistLimitDegrees=45
  twistMotion=Free
 }
 {
  bone1Name=dyn_hair_04_lf
  bone2Name=dyn_hair_03_lf
  swing1LimitDegrees=45
  swing1Motion=Free
  swing2LimitDegrees=45
  swing2Motion=Free
  twistLimitDegrees=45
  twistMotion=Free
 }
 {
  bone1Name=dyn_hair_05_lf
  bone2Name=dyn_hair_04_lf
  swing1LimitDegrees=45
  swing1Motion=Free
  swing2LimitDegrees=45
  swing2Motion=Free
  twistLimitDegrees=45
  twistMotion=Free
 }
 {
  bone1Name=dyn_hair_04_rt
  bone2Name=dyn_hair_03_rt
  swing1LimitDegrees=45
  swing1Motion=Free
  swing2LimitDegrees=45
  swing2Motion=Free
  twistLimitDegrees=45
  twistMotion=Free
 }
 {
  bone1Name=dyn_hair_05_rt
  bone2Name=dyn_hair_04_rt
  swing1LimitDegrees=45
  swing1Motion=Free
  swing2LimitDegrees=45
  swing2Motion=Free
  twistLimitDegrees=45
  twistMotion=Free
 }
 {
  bone1Name=dyn_hair_03_lf
  bone2Name=dyn_hair_02_lf
  swing1LimitDegrees=45
  swing1Motion=Free
  swing2LimitDegrees=45
  swing2Motion=Free
  twistLimitDegrees=45
  twistMotion=Free
 }
 {
  bone1Name=dyn_hair_03_rt
  bone2Name=dyn_hair_02_rt
  swing1LimitDegrees=45
  swing1Motion=Free
  swing2LimitDegrees=45
  swing2Motion=Free
  twistLimitDegrees=45
  twistMotion=Free
 }
 {
  bone1Name=dyn_hair_03_cr
  bone2Name=dyn_hair_02_cr
  swing1LimitDegrees=45
  swing1Motion=Free
  swing2LimitDegrees=45
  swing2Motion=Free
  twistLimitDegrees=45
  twistMotion=Free
 }
]
